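# Translucent unlit hologram material for the Alfredo male pickaxe logo FX, usable on skeletal and Niagara meshes.
# Emissive tints pickaxe and armor paint-over masks, adding panning tech and cell noise over a cyber grid.
# Opacity combines a local-position sphere mask, sine-pulsed grid and stencil masks; a hologram wiggle function takes the grid as bias.
# Scalar lerp switches blend in a glider mask or replace the result with particle color.
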Material /BRCosmetics/Weapons/FORT_Melee/Pickaxe_Alfredo_Male/FX/M_AlfredoMale_Logo
graph {
 properties {
  bUsedWithMeshDeformer=true
  bUsedWithNiagaraMeshParticles=true
  bUsedWithSkeletalMesh=true
  blendMode=BLEND_Translucent
  pixelDepthOffsetMode=PDOM_Legacy
  shadingModel=MSM_Unlit
  usedShadingModels=Unlit
 }
 nodes [
  n1 = SphereMask(
   A=ComponentMask(None=MaterialFunctionCall() {materialFunction=&LocalPosition}."Local Position (Excluding Offsets)") {b=true, g=true}
   B=AppendVector(A=ScalarParameter() {defaultValue=8, parameterName=Mask_Position_X}, B=ScalarParameter() {defaultValue=60, parameterName=Mask_Position_Y})
   Radius=ScalarParameter() {defaultValue=25, parameterName=Mask_Radius}
   Hardness=ScalarParameter() {defaultValue=-0.9, parameterName=Mask_Hardness}
  )
  n2 = Clamp(None=Multiply(A=Power(Base=@n1) {constExponent=1.8}) {constB=8}) {maxDefault=0.8}
  n3 = MaterialFunctionCall() {materialFunction=&DebugTimeSine}
  n4 = ScalarParameter() {defaultValue=7, parameterName="U_Tiling(B)"}
  n5 = ScalarParameter() {defaultValue=7, parameterName="V_Tiling(B)"}
  n6 = TextureSampleParameter2D(UVs=Multiply(A=AppendVector(A=@n4, B=@n5), B=TextureCoordinate())) {
   parameterName=Texture_Input_B
   samplerType=SAMPLERTYPE_LinearColor
   texture=/Game/Gadgets/Assets/VinderTech_Umbrella/Umbrella_Cyber/Textures/TilingCyberGrid_M
  }
  n7 = Saturate(None=Multiply(A=LinearInterpolate(Alpha=@n3.Result) {constA=0.2, constB=2}, B=@n6.RGB))
  n8 = ScalarParameter() {defaultValue=1, parameterName=Stencil_MasterOpacity1isForTransparent}
  n9 = Power(Base=TextureSample() {texture=/Game/Characters/Player/Male/Medium/Bodies/M_MED_Alfredo_Armor/Textures/T_M_MED_Alfredo_Armor_Body_D}.R)
  n10 = ScalarParameter() {defaultValue=1, parameterName=Armor_Switch_Difuse}
  n11 = Constant() {r=1}
  n12 = TextureSample() {texture=/Game/Characters/Player/Male/Medium/Bodies/M_MED_Alfredo_Armor/FX/FX_Armor_Body_PaintOver}
  n13 = Power(Base=@n12.G)
  n14 = Add(A=LinearInterpolate(B=ScalarParameter() {defaultValue=8, parameterName=Armor_Mult}, Alpha=Multiply(A=@n13) {constB=6}), B=@n12.B)
  n15 = TextureSampleParameter2D() {parameterName=Texture_Input_A, texture=/BRCosmetics/Weapons/FORT_Melee/Pickaxe_Alfredo_Male/Textures/T_Pickaxe_Alfredo_Male_D}
  n16 = Saturate(None=Power(Base=@n15.B) {constExponent=4})
  n17 = Multiply(
   A=VectorParameter() {parameterName=Emissive Color, defaultValue=<rgba 0 0.8779 1 1>}.RGB
   B=ScalarParameter() {defaultValue=2.5, parameterName=Emissive Mult}
  )
  n18 = Multiply(A=Multiply(A=LinearInterpolate(A=@n10, B=@n11, Alpha=@n14), B=Add(A=@n15.B, B=@n16)), B=@n17) {constB=2}
  n19 = LinearInterpolate(A=@n18, B=ScalarParameter() {parameterName=Stencil_1isForSolid}, Alpha=Saturate(None=Multiply(A=@n9) {constB=44}))
  n20 = ParticleColor()
  LinearInterpolate(A=Add(A=Add(A=@n2, B=@n7), B=Multiply(A=@n19, B=@n8)), B=@n20.A, Alpha=ScalarParameter() {parameterName=Switch_MeshEdgeGlow})
  MaterialFunctionCall!(
   WPO Intensity A=ScalarParameter() {defaultValue=0.2, parameterName=WPO_Intensity_A}
   WPO Intensity B=ScalarParameter() {defaultValue=0.1, parameterName=WPO_Intensity_B}
   Time
   Bias=@n6.RGB
  ) {materialFunction=/Game/Effects/Fort_Effects/Effects/Atlas/Materials/MF_Atlas_HologramWiggle_01}
  n21 = Multiply(A=@n1) {constB=0.5}
  n22 = Multiply(
   A=AppendVector(A=ScalarParameter() {defaultValue=8, parameterName="U_Tiling(C)"}, B=ScalarParameter() {defaultValue=8, parameterName="V_Tiling(C)"})
   B=TextureCoordinate()
  )
  n23 = TextureSampleParameter2D!(
   UVs=MaterialFunctionCall(UVs=Panner(Coordinate=@n22) {speedX=0.5}, "Rotation Angle (0-1)"=Constant()) {materialFunction=&CustomRotator}.Rotated Values
   Level
   Apply View MipBias
  ) {mipValueMode=TMVM_MipLevel, parameterName=Texture_Input_C, texture=/Game/Effects/Fort_Effects/Textures/Tile/Tests/T_Tile_Panner_Mask}
  n24 = ScalarParameter() {defaultValue=0.1, group=Tech, parameterName=Tech Noise Add}
  n25 = Multiply(A=Power(Base=Add(A=@n23.G, B=@n24)), B=ScalarParameter() {defaultValue=2, parameterName=Texture_Multiply_C}) {constB=2}
  n26 = Multiply(A=@n17, B=@n6.RGB)
  n27 = TextureCoordinate() {uTiling=2, vTiling=2}
  n28 = TextureSample(UVs=Panner(Coordinate=@n27, Speed=Constant2Vector() {g=-0.25, r=-0.25})) {samplerType=SAMPLERTYPE_Grayscale, texture=/BRCosmetics/Weapons/FORT_Melee/Pickaxe_Cube_Ninja_Male/FX/T_CellNoise}
  n29 = TextureSample(UVs=Panner(Coordinate=@n27, Speed=Constant2Vector() {g=0.25}) {constCoordinate=1}) {samplerType=SAMPLERTYPE_Grayscale, texture=&T_CellNoise}
  n30 = Add(A=Multiply(A=Multiply(A=@n18, B=@n21)) {constB=2}, B=Multiply(A=Multiply(A=@n25, B=@n26), B=Multiply(A=@n28.RGB, B=@n29.RGB)))
  n31 = TextureSample() {texture=/Game/Gadgets/Assets/VinderTech_GliderChute/Glider_Alfredo_Male/FX/T_Glider_Alfredo_Male_FX_RChannelOnly}
  n32 = ScalarParameter() {defaultValue=1, parameterName=Glider}
  n33 = LinearInterpolate(A=Subtract(A=@n30, B=Power(Base=@n31.R) {constExponent=0.7}), Alpha=@n32)
  MaterialFunctionCall(Emissive=LinearInterpolate(A=LinearInterpolate(A=@n33, B=@n30, Alpha=@n32), B=@n20.RGB, Alpha=ScalarParameter() {parameterName=Switch_MeshEdgeGlow})) {materialFunction=&MF_FX_EmissiveAdjustments}
  TextureSampleParameter2D!(UVs, Apply View MipBias) {parameterName=Texture_Input_D, texture=/Game/Gadgets/Assets/VinderTech_GliderChute/Glider_Alfredo_Male/Textures/T_Glider_Alfredo_Male_FX}
 ]
}
